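# Physics asset for the ClimbSpill balloon gadget, simulating a balloon on a flexible string.
# A sphere body forms the balloon, and a chain of small capsules runs from string_base up through string_08.
# String joints leave both swings free and lock twist, so the string bends like a rope without spinning along its length.
# The balloon's joint to string_08 locks all angular motion, so it moves rigidly with the string's tip.
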PhysicsAsset /WildEstateWeapons/Assets/ClimbSpill_Balloon/ClimbSpill_Balloon_Gadget_Physics
bodies {
 balloon [{shapeType=Sphere, radius=32, shapeName=balloon_sphere, center=<0.034 0.0258 2.3134>}]
 string_01 [{shapeType=Capsule, length=5, radius=1, shapeName=string_01_capsule, rotation=<rot 90 0 0>}]
 string_02 [{shapeType=Capsule, length=5, radius=1, shapeName=string_02_capsule, rotation=<rot 90 0 0>}]
 string_03 [{shapeType=Capsule, length=5, radius=1, shapeName=string_03_capsule, rotation=<rot 90 0 0>}]
 string_04 [{shapeType=Capsule, length=5, radius=1, shapeName=string_04_capsule, rotation=<rot 90 0 0>}]
 string_05 [{shapeType=Capsule, length=5, radius=1, shapeName=string_05_capsule, rotation=<rot 90 0 0>}]
 string_06 [{shapeType=Capsule, length=5, radius=1, shapeName=string_06_capsule, rotation=<rot 90 0 0>}]
 string_07 [{shapeType=Capsule, length=5, radius=1, shapeName=string_07_capsule, rotation=<rot 90 0 0>}]
 string_08 [{shapeType=Capsule, length=5, radius=1, shapeName=string_08_capsule, center=<-3.071 0 0>, rotation=<rot 90 0 0>}]
 string_base [{shapeType=Capsule, length=5, radius=1, shapeName=string_base_capsule, rotation=<rot 90 0 0>}]
}
constraints [
 {
  bone1Name=string_08
  bone2Name=string_07
  swing1LimitDegrees=45
  swing1Motion=Free
  swing2LimitDegrees=45
  swing2Motion=Free
  twistLimitDegrees=45
  twistMotion=Locked
 }
 {
  bone1Name=string_01
  bone2Name=string_base
  swing1LimitDegrees=45
  swing1Motion=Free
  swing2LimitDegrees=45
  swing2Motion=Free
  twistLimitDegrees=45
  twistMotion=Locked
 }
 {
  bone1Name=string_02
  bone2Name=string_01
  swing1LimitDegrees=45
  swing1Motion=Free
  swing2LimitDegrees=45
  swing2Motion=Free
  twistLimitDegrees=45
  twistMotion=Locked
 }
 {
  bone1Name=string_03
  bone2Name=string_02
  swing1LimitDegrees=45
  swing1Motion=Free
  swing2LimitDegrees=45
  swing2Motion=Free
  twistLimitDegrees=45
  twistMotion=Locked
 }
 {
  bone1Name=string_04
  bone2Name=string_03
  swing1LimitDegrees=45
  swing1Motion=Free
  swing2LimitDegrees=45
  swing2Motion=Free
  twistLimitDegrees=45
  twistMotion=Locked
 }
 {
  bone1Name=string_05
  bone2Name=string_04
  swing1LimitDegrees=45
  swing1Motion=Free
  swing2LimitDegrees=45
  swing2Motion=Free
  twistLimitDegrees=45
  twistMotion=Locked
 }
 {
  bone1Name=string_06
  bone2Name=string_05
  swing1LimitDegrees=45
  swing1Motion=Free
  swing2LimitDegrees=45
  swing2Motion=Free
  twistLimitDegrees=45
  twistMotion=Locked
 }
 {
  bone1Name=string_07
  bone2Name=string_06
  swing1LimitDegrees=45
  swing1Motion=Free
  swing2LimitDegrees=45
  swing2Motion=Free
  twistLimitDegrees=45
  twistMotion=Locked
 }
 {
  bone1Name=balloon
  bone2Name=string_08
  swing1LimitDegrees=45
  swing1Motion=Locked
  swing2LimitDegrees=45
  swing2Motion=Locked
  twistLimitDegrees=45
  twistMotion=Locked
 }
]
